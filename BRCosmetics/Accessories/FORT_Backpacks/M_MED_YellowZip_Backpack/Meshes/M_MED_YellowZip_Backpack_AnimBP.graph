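# Animation Blueprint for the yellow-zip backpack cosmetic that adds secondary motion to its zipper bones.
# The local-space ref pose runs through four chained Anim Dynamics nodes (top zippers B and A, upper, lower), then returns to local space.
# Each applies angular limits, overridden high angular damping and component linear motion scaling.
# The handle bone's dynamics node is unconnected, and the event graph holds only unconnected default nodes.
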AnimBlueprint /BRCosmetics/Accessories/FORT_Backpacks/M_MED_YellowZip_Backpack/Meshes/M_MED_YellowZip_Backpack_AnimBP
target_skeleton=&/M_MED_YellowZip_Backpack_Skeleton
graphs [
 AnimGraph {nodes [
   Anim Dynamics - Bone: dyn_handle(Alpha=1, ExternalForce="0,0,0") {
    bGravityOverrideInSimSpace=true
    alphaBoolBlend {blendOption=Linear}
    boundBone {boneName=dyn_handle}
    physicsBodyDefinitions [{
      sphereCollisionRadius=0
      boundBone {boneName=dyn_handle}
      boxExtents=<5 30 30>
      constraintSetup {angularLimitsMax=<0 20 10>, angularLimitsMin=<0 -20 -10>, angularTarget=<0 0 0>}
      localJointOffset=<-5 0 0>
     }]
   }
   n1 = Anim Dynamics - Bone: dyn_top_zipperB(ComponentPose=Local To Component(LocalPose=Local Space Ref Pose()), Alpha=1, ExternalForce="0,0,0") {
    angularDampingOverride=0.98
    bGravityOverrideInSimSpace=true
    bOverrideAngularDamping=true
    alphaBoolBlend {blendOption=Linear}
    boundBone {boneName=dyn_top_zipperB}
    componentLinearAccScale=<1 1 1>
    componentLinearVelScale=<1 1 1>
    physicsBodyDefinitions [{
      sphereCollisionRadius=0
      boundBone {boneName=dyn_top_zipperB}
      boxExtents=<5 5 5>
      constraintSetup {angularLimitsMax=<0 40 120>, angularLimitsMin=<0 -40 0>, angularTarget=<0 0 0>}
      localJointOffset=<-1 0 0>
     }]
   }
   n2 = Anim Dynamics - Bone: dyn_top_zipperA(ComponentPose=@n1, Alpha=1, ExternalForce="0,0,0") {
    angularDampingOverride=0.98
    bGravityOverrideInSimSpace=true
    bOverrideAngularDamping=true
    alphaBoolBlend {blendOption=Linear}
    boundBone {boneName=dyn_top_zipperA}
    componentLinearAccScale=<1 1 1>
    componentLinearVelScale=<1 1 1>
    physicsBodyDefinitions [{
      sphereCollisionRadius=0
      boundBone {boneName=dyn_top_zipperA}
      boxExtents=<5 4 5>
      constraintSetup {angularLimitsMax=<0 40 120>, angularLimitsMin=<0 -40 0>, angularTarget=<0 0 0>}
      localJointOffset=<-1.5 0 0>
     }]
   }
   n3 = Anim Dynamics - Bone: dyn_upper_zipper(ComponentPose=@n2, Alpha=1, ExternalForce="0,0,0") {
    angularDampingOverride=0.98
    bGravityOverrideInSimSpace=true
    bOverrideAngularDamping=true
    alphaBoolBlend {blendOption=Linear}
    boundBone {boneName=dyn_upper_zipper}
    componentLinearAccScale=<1 1 1>
    componentLinearVelScale=<1 1 1>
    physicsBodyDefinitions [{
      sphereCollisionRadius=0
      boundBone {boneName=dyn_upper_zipper}
      boxExtents=<4 5 4>
      constraintSetup {angularLimitsMax=<0 40 120>, angularLimitsMin=<0 -40 0>, angularTarget=<0 0 0>}
      localJointOffset=<-1 0 0>
     }]
   }
   n4 = Anim Dynamics - Bone: dyn_lower_zipper(ComponentPose=@n3, Alpha=1, ExternalForce="0,0,0") {
    angularDampingOverride=0.98
    bGravityOverrideInSimSpace=true
    bOverrideAngularDamping=true
    alphaBoolBlend {blendOption=Linear}
    boundBone {boneName=dyn_lower_zipper}
    componentLinearAccScale=<1 1 1>
    componentLinearVelScale=<1 1 1>
    physicsBodyDefinitions [{
      sphereCollisionRadius=0
      boundBone {boneName=dyn_lower_zipper}
      boxExtents=<4 4.5 5>
      constraintSetup {angularLimitsMax=<0 40 120>, angularLimitsMin=<0 -40 0>, angularTarget=<0 0 0>}
      localJointOffset=<-1.5 0 0>
     }]
   }
   Output Pose(Result=Component To Local(ComponentPose=@n4)) {layerGroup=None}
  ]}
 EventGraph {nodes [
   TryGetPawnOwner()
   Event BlueprintUpdateAnimation()
  ]}
]
variables []
